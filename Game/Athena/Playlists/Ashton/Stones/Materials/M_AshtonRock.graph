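# M_AshtonRock is an opaque, metallic rock material for the Ashton playlist stones, with tinted inner glow, tiled cave-wall detail normals and noise-driven roughness.
# World position offset spins the mesh about Z at its component pivot and bobs it on a sine, with normals corrected for the rotation.
# Emissive fakes interior depth by combining a view-facing falloff, panning pulsing shatter masks and noise, confined by a sphere mask around the object center.
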Material /Game/Athena/Playlists/Ashton/Stones/Materials/M_AshtonRock
graph {
 properties {bUseMaterialAttributes=true, pixelDepthOffsetMode=PDOM_Legacy, usedShadingModels=Default Lit}
 nodes [
  ActorPositionWS()
  Constant3Vector() {constant=<rgba 1 0.8034 0.1009 1>}
  Constant3Vector() {constant=<rgba 0.5 0.4531 0.05 1>}
  Constant3Vector() {constant=<rgba 0.5 0.203 0.05 1>}
  Constant3Vector() {constant=<rgba 0 0.9 0.1 1>}
  Constant3Vector() {constant=<rgba 0 0.3 0.06 1>}
  Constant3Vector() {constant=<rgba 0.02 0.265 1 1>}
  Constant3Vector() {constant=<rgba 0.09 0.1954 0.6 1>}
  Constant3Vector() {constant=<rgba 0.5862 0.0928 1 1>}
  Constant3Vector() {constant=<rgba 0.3409 0.2 0.5 1>}
  Constant3Vector() {constant=<rgba 1 0.1 0.145 1>}
  Constant3Vector() {constant=<rgba 0.5 0.15 0.15 1>}
  Constant3Vector() {constant=<rgba 1 0.29 0.05 1>}
  Constant() {r=-0.1}
  Constant() {r=4}
  Constant() {r=0.1}
  n1 = VectorParameter() {parameterName="DM-RG=Tile-BA=Offset", defaultValue=<rgba 1.5 4 0.2 0>}
  n2 = Add(A=Multiply(A=TextureCoordinate(), B=ComponentMask(None=@n1.RGB) {g=true, r=true}), B=AppendVector(A=@n1.B, B=@n1.B))
  n3 = Constant3Vector() {constant=<rgba 0 0 1 0>}
  n4 = Time()
  n5 = Multiply(A=Multiply(A=ScalarParameter() {defaultValue=0.1, parameterName=MeshRotationSpeed}, B=Constant3Vector() {constant=<rgba 1 0 0 0>}), B=@n4)
  n6 = MaterialFunctionCall(Rotation Axis=@n3, Rotation Angle=@n5) {materialFunction=&FixRotateAboutAxisNormals}
  n7 = MaterialFunctionCall(
   BaseNormal=@n6.New Tangent Space Vertex Normal
   AdditionalNormal=Multiply(
    A=Constant3Vector() {constant=<rgba -1 -1 1 0>}
    B=TextureSample(UVs=@n2) {samplerType=SAMPLERTYPE_Normal, texture=/Game/Packages/DS_Fortnite_Terrain_NoLOD/Textures/T_Cave_Exterior_Wall_N}.RGB
   )
  ) {materialFunction=&BlendAngleCorrectedNormals}
  LinearInterpolate(A=@n7.Result, B=@n6.New Tangent Space Vertex Normal, Alpha=ScalarParameter() {parameterName=DM-DetailNormalAmount})
  n8 = MaterialFunctionCall() {materialFunction=&ComponentPivotLocation}
  n9 = WorldPosition()
  n10 = Sine(None=@n4) {period=3.3}
  n11 = Constant3Vector() {constant=<rgba 0 0 3 0>}
  n12 = TextureSample(UVs=Multiply(A=TextureCoordinate(), B=ScalarParameter() {defaultValue=1, parameterName=RoughTile})) {samplerType=SAMPLERTYPE_Masks, texture=/Game/Global/Textures/Default/Noise/T_CubeNoise}
  n13 = Power(
   Base=DotProduct(A=CameraVectorWS(), B=@n6.New World Space Vertex Normal)
   Exp=ScalarParameter() {defaultValue=5, parameterName=InteriorShard-EdgeFalloff}
  ) {constExponent=8}
  n14 = TextureCoordinate()
  n15 = ScalarParameter() {defaultValue=4, parameterName=InteriorShard-Tile}
  n16 = Time()
  n17 = ScalarParameter() {defaultValue=-0.08, parameterName=InteriorShard-PanSpeed}
  n18 = TextureSample(UVs=Panner(Coordinate=Multiply(A=@n14, B=@n15), Time=Multiply(A=@n16, B=@n17)) {speedX=1}) {texture=/Game/Weapons/WeaponSkins/Wraps/Textures/ShatterFly/ShatterFly_Glass_Mask}
  n19 = ScalarParameter() {defaultValue=5, parameterName=InteriorShard-GradientAmount}
  n20 = Add(A=Add(A=@n18.B, B=Constant2Vector()), B=Multiply(A=Time(), B=ScalarParameter() {defaultValue=0.4, parameterName=InteriorShard-PulseSpeed}))
  n21 = Multiply(
   A=Saturate(None=Multiply(A=@n18.G, B=@n19))
   B=Multiply(
    A=TextureSample(UVs=@n20) {texture=/Game/Global/Textures/Default/Masks_and_Gradients/T_VertGradient_02_sRGB}.R
    B=ScalarParameter() {defaultValue=5, parameterName=InteriorShard-PulseAmount}
   )
  )
  n22 = SphereMask(
   A=WorldPosition() {worldPositionShaderOffset=WPT_ExcludeAllShaderOffsets}.XYZ
   B=ObjectPositionWS()
   Radius=ScalarParameter() {defaultValue=20, parameterName=InteriorGlowSM-Radius}
   Hardness=ScalarParameter() {defaultValue=0.25, parameterName=InteriorGlowSM-Hardness}
  )
  n23 = Multiply(A=Multiply(A=@n13, B=Add(A=@n18.G, B=@n21)), B=Multiply(A=ScalarParameter() {defaultValue=20, parameterName=InteriorGlowSM-Mod1}, B=@n22))
  n24 = ScalarParameter() {defaultValue=5, parameterName=InteriorGlowSM-Mod2}
  n25 = Power(Base=@n22, Exp=ScalarParameter() {defaultValue=2, parameterName=InteriorGlowSM-Power})
  n26 = Multiply(A=@n12.G, B=Multiply(A=@n13, B=Multiply(A=@n24, B=@n25)))
  MakeMaterialAttributes!(
   BaseColor=VectorParameter() {parameterName=BaseColor, defaultValue=<rgba 0.25 0.25 0.25 1>}.RGB
   Metallic=ScalarParameter() {defaultValue=0.9, parameterName=Metal}
   Specular=ScalarParameter() {defaultValue=0.8, parameterName=Spec}
   Roughness=LinearInterpolate(
    A=ScalarParameter() {defaultValue=0.2, parameterName=RoughLowAmount}
    B=ScalarParameter() {defaultValue=0.6, parameterName=RoughHighAmount}
    Alpha=Multiply(A=@n12.R, B=@n12.R)
   )
   Anisotropy
   EmissiveColor=Multiply(A=VectorParameter() {parameterName=EmissiveColor, defaultValue=<rgba 1 1 1 1>}.RGB, B=Add(A=@n23, B=@n26))
   Opacity
   OpacityMask
   Normal=@n7.Result
   Tangent
   WorldPositionOffset=Add(A=RotateAboutAxis(NormalizedRotationAxis=@n3, RotationAngle=@n5, PivotPoint=@n8.Result, Position=@n9.XYZ), B=Multiply(A=@n10, B=@n11))
   SubsurfaceColor
   ClearCoat
   ClearCoatRoughness
   AmbientOcclusion
   Refraction
   CustomizedUVs_0
   CustomizedUVs_1
   CustomizedUVs_2
   CustomizedUVs_3
   CustomizedUVs_4
   CustomizedUVs_5
   CustomizedUVs_6
   CustomizedUVs_7
   PixelDepthOffset
   ShadingModel
   Displacement
  )
 ]
}
